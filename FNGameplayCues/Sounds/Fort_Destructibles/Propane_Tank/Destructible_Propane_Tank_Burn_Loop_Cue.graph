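# Sound cue for the looping burn sound of a destructible propane tank.
# One looping wave passes through a gadgets sound class override, then a modulator that randomizes pitch upward at a fixed reduced volume.
# An attenuation node, based on the shared RPG explosion attenuation asset, overrides settings to add distance low-pass filtering, reverb wet level and a custom shape extent.
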SoundCue /FNGameplayCues/Sounds/Fort_Destructibles/Propane_Tank/Destructible_Propane_Tank_Burn_Loop_Cue
root=Attenuation_0
nodes [
 {
  id=Attenuation_0
  children [Modulator_1]
  properties {
   attenuationSettings=/Game/Sounds/Attenuation/Fort_Explosion_RPG_Att
   bOverrideAttenuation=true
   attenuationOverrides {
    bAttenuateWithLPF=true
    lPFFrequencyAtMax=665.3513
    lPFRadiusMax=2000
    lPFRadiusMin=1200
    reverbWetLevelMin=0.2
    attenuationShapeExtents {x=1407.6522}
   }
  }
 }
 {id=Modulator_1, children [SoundClass_1], properties {pitchMax=1.3, pitchMin=1, volumeMax=0.4, volumeMin=0.4}}
 {id=SoundClass_1, children [WavePlayer_0], properties {soundClassOverride=&Fort_Gadgets}}
 {
  id=WavePlayer_0
  properties {bLooping=true, soundWaveAssetPtr=&/Destructible_Propane_Tank_Burn_Loop}
 }
]
properties {maxDistance=5007.6523, soundClassObject=&Fort_Gadgets_Items, volumeMultiplier=1.3393}
